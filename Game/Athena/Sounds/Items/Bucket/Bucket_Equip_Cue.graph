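# Sound Cue for equipping the Bucket item, layering an equip sound with a quieter in-air whoosh.
# A Mixer blends the boosted equip wave with a random pick of three equally weighted projectile whoosh waves.
# Each layer has its own Modulator for pitch, with the whoosh shifted noticeably higher.
# The cue plays in the player foley sound class with weapon pickup attenuation.
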SoundCue /Game/Athena/Sounds/Items/Bucket/Bucket_Equip_Cue
root=Mixer_1
nodes [
 {id=Mixer_1, children [Modulator_0,Modulator_1], properties {inputVolume [2,0.3]}}
 {id=Modulator_0, children [WavePlayer_5], properties {pitchMin=1, volumeMax=1, volumeMin=1}}
 {id=Modulator_1, children [Random_0], properties {pitchMax=1.3, pitchMin=1.2, volumeMax=1, volumeMin=1}}
 {id=Random_0, children [WavePlayer_1,WavePlayer_3,WavePlayer_4], properties {bShouldExcludeFromBranchCulling=true, weights [1,1,1]}}
 {id=WavePlayer_1, properties {soundWaveAssetPtr=&/Bucket_Nice_Prj_InAir_Whoosh_01}}
 {id=WavePlayer_3, properties {soundWaveAssetPtr=&/Bucket_Nice_Prj_InAir_Whoosh_02}}
 {id=WavePlayer_4, properties {soundWaveAssetPtr=&/Bucket_Nice_Prj_InAir_Whoosh_03}}
 {id=WavePlayer_5, properties {soundWaveAssetPtr=&/Bucket_Nice_Equip_01}}
]
properties {
 attenuationSettings=&Fort_Weapon_Pickup_Att
 maxDistance=3500
 soundClassObject=&Fort_Player_Foley
 volumeMultiplier=1.5
}
