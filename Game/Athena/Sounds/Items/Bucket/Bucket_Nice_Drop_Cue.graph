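# Sound cue for the drop sound of the Nice Bucket item in Athena.
# A single wave player feeds the Bucket_Nice_Drop wave straight to the output, with no randomization or modulation.
# It uses the shared weapon-drop attenuation and the loot-collect sound class, with a limited audible range.
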SoundCue /Game/Athena/Sounds/Items/Bucket/Bucket_Nice_Drop_Cue
root=WavePlayer_0
nodes [{id=WavePlayer_0, properties {soundWaveAssetPtr=&/Bucket_Nice_Drop}}]
properties {
 attenuationSettings=&Fort_Weapon_Drop_Att
 maxDistance=3000
 soundClassObject=&Fort_Loot_Collect
}
