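# Sound cue for the bucket item, named for a "nice land", that contains no nodes and has no root, so it plays no audio.
# It still sets a weapon-drop attenuation asset, a loot-collect sound class and a maximum audible distance.
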SoundCue /Game/Athena/Sounds/Items/Bucket/Bucket_Nice_Land_Cue
root=null
nodes []
properties {
 attenuationSettings=&Fort_Weapon_Drop_Att
 maxDistance=3000
 soundClassObject=&Fort_Loot_Collect
}
