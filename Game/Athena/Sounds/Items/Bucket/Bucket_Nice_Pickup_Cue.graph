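# Sound cue for picking up the Bucket_Nice item in Athena.
# A single wave player feeds the Bucket_Nice_Pickup sound straight to the output, with no randomization or mixing.
# It plays at reduced volume in the gadget items sound class, spatialized through the shared weapon pickup attenuation.
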SoundCue /Game/Athena/Sounds/Items/Bucket/Bucket_Nice_Pickup_Cue
root=WavePlayer_0
nodes [{id=WavePlayer_0, properties {soundWaveAssetPtr=&/Bucket_Nice_Pickup}}]
properties {
 attenuationSettings=&Fort_Weapon_Pickup_Att
 maxDistance=3500
 soundClassObject=&Fort_Gadgets_Items
 volumeMultiplier=0.5
}
