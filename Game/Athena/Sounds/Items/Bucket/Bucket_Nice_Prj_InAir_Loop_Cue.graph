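# Sound cue for the in-flight audio of the "Nice" Bucket item projectile, mixing a continuous loop with a whoosh layer.
# The loop fades in through a volume envelope, then passes through a modulator, smoothed Doppler and its own attenuation.
# The whoosh picks randomly from three equally weighted waves, with its own modulator and separate attenuation.
# A concurrency override allows two instances at once, stopping the oldest.
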SoundCue /Game/Athena/Sounds/Items/Bucket/Bucket_Nice_Prj_InAir_Loop_Cue
root=Mixer_2
nodes [
 {id=Attenuation_4, children [Modulator_4], properties {attenuationSettings=&/Bucket_Nice_Prj_InAir_Whoosh_Att}}
 {id=Attenuation_5, children [Doppler_0], properties {attenuationSettings=&/Bucket_Nice_Prj_InAir_Loop_Att}}
 {id=Doppler_0, children [Modulator_5], properties {bUseSmoothing=true, dopplerIntensity=0.5}}
 {
  id=Enveloper_4
  children [WavePlayer_6]
  properties {pitchCurve {editorCurveData {keys [{time=0}]}}, volumeCurve {editorCurveData {keys [<at 0 0 RCIM_Cubic>,<at 0.5 1>]}}}
 }
 {id=Mixer_2, children [Attenuation_5, Attenuation_4], properties {inputVolume [0.75,1]}}
 {id=Modulator_4, children [Random_0]}
 {id=Modulator_5, children [Enveloper_4]}
 {id=Random_0, children [WavePlayer_3,WavePlayer_4,WavePlayer_5], properties {bShouldExcludeFromBranchCulling=true, weights [1,1,1]}}
 {id=WavePlayer_3, properties {soundWaveAssetPtr=&/Bucket_Nice_Prj_InAir_Whoosh_01}}
 {id=WavePlayer_4, properties {soundWaveAssetPtr=&/Bucket_Nice_Prj_InAir_Whoosh_02}}
 {id=WavePlayer_5, properties {soundWaveAssetPtr=&/Bucket_Nice_Prj_InAir_Whoosh_03}}
 {id=WavePlayer_6, properties {bLooping=true, soundWaveAssetPtr=&/Bucket_Nice_Prj_InAir_Loop}}
]
properties {bOverrideConcurrency=true, maxDistance=20000, volumeMultiplier=1, concurrencyOverrides {maxCount=2, resolutionRule=StopOldest}}
