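# Sound cue that plays the Bucket_Nice_Throw sound for the Bucket item in Athena.
# A single wave player feeds a modulator at the root, which keeps its default settings.
# The cue uses the gadget-placement attenuation asset and is assigned to the gadgets and items sound class.
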SoundCue /Game/Athena/Sounds/Items/Bucket/Bucket_Nice_Throw_Cue
root=Modulator_0
nodes [
 {id=Modulator_0, children [WavePlayer_1]}
 {id=WavePlayer_1, properties {soundWaveAssetPtr=&/Bucket_Nice_Throw}}
]
properties {
 attenuationSettings=/Game/Sounds/Attenuation/Fort_GadgetPlace_Att
 maxDistance=2000
 soundClassObject=&Fort_Gadgets_Items
 volumeMultiplier=1
}
